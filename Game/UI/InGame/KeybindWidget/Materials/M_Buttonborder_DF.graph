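# Translucent UI material that draws a keybind widget button shape with a colored border and a separately tinted interior.
# The red channel of T_Buttonborder_DF, sampled with center-scaled UVs, acts as a gradient for two AAThreshold calls, giving anti-aliased edges.
# BorderThreshold sets the outer edge, and adding BorderSize sets the inner edge that blends BorderColor into InteriorColor.
# Opacity is the outer mask times border alpha, minus the inner mask times inverted interior alpha.
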Material /Game/UI/InGame/KeybindWidget/Materials/M_Buttonborder_DF
graph {
 properties {
  bUsedWithStaticMesh=false
  blendMode=BLEND_Translucent
  materialDomain=MD_UI
  pixelDepthOffsetMode=PDOM_Legacy
  usedShadingModels=Default Lit
 }
 nodes [
  n1 = VectorParameter() {parameterName=BorderColor, defaultValue=<rgba 1 1 1 1>}
  n2 = VectorParameter() {parameterName=InteriorColor, defaultValue=<rgba 0.0052 0.0052 0.0056 0.4>}
  n3 = ScalarParameter() {defaultValue=0.6, parameterName=BorderThreshold}
  n4 = TextureSample(UVs=MaterialFunctionCall(Texture Scale=ScalarParameter() {defaultValue=1, parameterName=Scale}) {materialFunction=&ScaleUVsByCenter}.UVs) {texture=/FortUILibrary/Input/Misc/T_Buttonborder_DF}
  n5 = MaterialFunctionCall!(
   Threshold=Add(A=@n3, B=ScalarParameter() {defaultValue=0.1, parameterName=BorderSize})
   Gradient=@n4.R
   Filter Width Override
   Filter Width Fudge Factor
  ) {materialFunction=/Game/UI/Materials/AAThreshold}
  LinearInterpolate(A=@n1.RGB, B=@n2.RGB, Alpha=@n5.Result)
  Subtract(
   A=Multiply(A=MaterialFunctionCall(Threshold=@n3, Gradient=@n4.R) {materialFunction=&AAThreshold}.Result, B=@n1.A)
   B=Multiply(A=OneMinus(None=@n2.A), B=@n5.Result)
  )
 ]
}
