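# Keybind widget, derived from KeybindWidget, that displays the input action supplied by a global action details data source.
# SetActionSource swaps sources only when changed, moving the OnChange delegate binding to the new one and refreshing.
# UpdateInputAction caches the source's ActionDetails, then sets the bound action name and maps hold status to forced-hold or no forced hold.
# An invalid source reapplies the stored details rather than clearing them.
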WidgetBlueprint /Game/UI/InGame/KeybindWidget/ActionDetailsSourceKeybindWidget
graphs [
 BindActionSourceEvents {nodes [
   BindActionSourceEvents() {is=title K2Node_FunctionEntry}
   n1 = Get ActionSource() -> (ActionSource:Fort Global Action Details Data Source Interface)
   > Is Valid(InputObject="To Object (Interface)"!(Interface:"Interface Interface (by ref)"=@n1) -> (ReturnValue:Object Reference))
   > AddOnChangeDelegate!(execute, self:Fort Global Action Details Data Source Interface=@n1, InDelegate:"Delegate (by ref)"=Create Event()) -> (then)
  ]}
 EventGraph {nodes [
   Event PreConstruct()
   Event Construct()
   Event Tick()
  ]}
 HandleActionSourceValueChange {nodes [
   HandleActionSourceValueChange!() -> (then, Source:Object Reference) {is=title K2Node_FunctionEntry}
   > UpdateInputAction()
  ]}
 SetActionSource {nodes [
   n1 = SetActionSource!() -> (then, Value:Fort Global Action Details Data Source Interface) {is=title K2Node_FunctionEntry}
   n2 = Get ActionSource()
   > Branch(Condition="Not Equal (Object)"!(A:Object Reference="To Object (Interface)"(Interface=@n2), B:Object Reference="To Object (Interface)"(Interface=@n1)) -> (ReturnValue:bool))
   > UnbindActionSourceEvents()
   > Set ActionSource(ActionSource:Fort Global Action Details Data Source Interface=@n1) -> (Output_Get:Fort Global Action Details Data Source Interface)
   > BindActionSourceEvents() {is=title K2Node_CallFunction}
   > UpdateInputAction()
  ]}
 UnbindActionSourceEvents {nodes [
   UnbindActionSourceEvents() {is=title K2Node_FunctionEntry}
   n1 = Get ActionSource()
   > Is Valid(InputObject="To Object (Interface)"(Interface=@n1))
   > RemoveOnChangeDelegate!(execute, self:Fort Global Action Details Data Source Interface=@n1, InDelegate:"Delegate (by ref)"=Create Event()) -> (then)
  ]}
 UpdateInputAction {nodes [
   UpdateInputAction() {is=title K2Node_FunctionEntry}
   n1 = Get ActionSource()
   > n2 = Is Valid(InputObject="To Object (Interface)"(Interface=@n1))
   > n3 = Set ActionDetails(ActionDetails:Fort Global Action Details Structure=GetValue!(self:Fort Global Action Details Data Source Interface=@n1) -> (ReturnValue:Fort Global Action Details Structure)) -> (Output_Get:Fort Global Action Details Structure)
   n4 = Get!() -> (ActionDetails_ActionName:Name, ActionDetails_HoldStatus:bool=~)
   SetBoundAction!(execute=[@n3.then, @n2.Is Not Valid], self:Fort Keybind Widget Object Reference, NewBoundAction:Name=@n4) -> (then)
   > SetForcedHoldKeybindStatus!(
    execute
    self:Fort Keybind Widget Object Reference
    InForcedHoldKeybindStatus:EFortKeybindForcedHoldStatus Enum=Select(
     Option 0:EFortKeybindForcedHoldStatus Enum=NoForcedHold
     Option 1:EFortKeybindForcedHoldStatus Enum=ForcedHold
     Index=@n4.ActionDetails_HoldStatus
    ) -> (ReturnValue:EFortKeybindForcedHoldStatus Enum)
   ) -> (then)
  ]}
]
variables [ActionSource {default=None}]
widgets {
 info {inheritedWidgetCount=0, parentClass=/FortUILibrary/Input/KeybindWidget.KeybindWidget_C, rootWidgetClass=null, widgetCount=0}
 widgets []
}
