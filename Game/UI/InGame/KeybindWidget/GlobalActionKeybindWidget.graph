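# Keybind widget for in-game UI that shows the binding for a configurable global action, as a child of ActionDetailsSourceKeybindWidget with no widgets of its own.
# On initialization it constructs a standard global-action details data source, pushes its stored action and context into it, then sets it as the parent's action source.
# SetGlobalAction and SetActionDetailsContext store new values and forward them to that source through Propagate functions, without null checks.
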WidgetBlueprint /Game/UI/InGame/KeybindWidget/GlobalActionKeybindWidget
graphs [
 EventGraph {nodes [
   Event OnInitialized()
   > Parent: OnInitialized() {is=title K2Node_CallParentFunction}
   n1 = Self-Reference()
   > n2 = "Construct FortStandardGlobalActionDetailsDataSource (with default contexts)"!(
    execute
    Outer:Object Reference=@n1
    ContextWidget:Widget Object Reference=@n1
    GlobalAction:EFortGlobalAction Enum=TrapConfirm
    ActionDetailsContext:"Fort Global Action Details Function Context Structure (by ref)"
   ) -> (then, ReturnValue:Fort Standard Global Action Details Data Source Object Reference)
   > n3 = Set TypedActionSource(TypedActionSource:Fort Standard Global Action Details Data Source Object Reference=@n2) -> (Output_Get:Fort Standard Global Action Details Data Source Object Reference)
   > PropagateActionDetailsContext()
   > PropagateGlobalAction()
   > SetActionSource!(execute, self:Action Details Source Keybind Widget Object Reference, Value:Fort Global Action Details Data Source Interface=@n3) -> (then)
   Event PreConstruct()
   Event Construct()
   Event Tick()
  ]}
 PropagateActionDetailsContext {nodes [
   PropagateActionDetailsContext() {is=title K2Node_FunctionEntry}
   > SetActionDetailsContext!(
    execute
    self:Fort Standard Global Action Details Data Source Object Reference=Get TypedActionSource() -> (TypedActionSource:Fort Standard Global Action Details Data Source Object Reference)
    Value:"Fort Global Action Details Function Context Structure (by ref)"=Get ActionDetailsContext() -> (ActionDetailsContext:Fort Global Action Details Function Context Structure)
   ) -> (then)
  ]}
 PropagateGlobalAction {nodes [
   PropagateGlobalAction() {is=title K2Node_FunctionEntry}
   > SetGlobalAction!(
    execute
    self:Fort Standard Global Action Details Data Source Object Reference=Get TypedActionSource()
    Value:EFortGlobalAction Enum=Get GlobalAction() -> (GlobalAction:EFortGlobalAction Enum=TrapConfirm)
   ) -> (then)
  ]}
 SetActionDetailsContext {nodes [
   n1 = SetActionDetailsContext!() -> (then, Value:Fort Global Action Details Function Context Structure) {is=title K2Node_FunctionEntry}
   > Set ActionDetailsContext(ActionDetailsContext:Fort Global Action Details Function Context Structure=@n1) -> (Output_Get:Fort Global Action Details Function Context Structure)
   > PropagateActionDetailsContext() {is=title K2Node_CallFunction}
  ]}
 SetGlobalAction {nodes [
   n1 = SetGlobalAction!() -> (then, Value:EFortGlobalAction Enum=TrapConfirm) {is=title K2Node_FunctionEntry}
   > Set GlobalAction(GlobalAction:EFortGlobalAction Enum=@n1) -> (Output_Get:EFortGlobalAction Enum=TrapConfirm)
   > PropagateGlobalAction() {is=title K2Node_CallFunction}
  ]}
]
variables [
 ActionDetailsContext {type=FortGlobalActionDetailsFunctionContext, default {overrideInputType=Count}}
 GlobalAction {type=EFortGlobalAction, default=Count}
 TypedActionSource {type=FortStandardGlobalActionDetailsDataSource, default=null}
]
widgets {
 info {
  inheritedWidgetCount=0
  parentClass=&/ActionDetailsSourceKeybindWidget.ActionDetailsSourceKeybindWidget_C
  rootWidgetClass=null
  widgetCount=0
 }
 widgets []
}
